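; Glider Blueprint for the generic Build-A-Brella cosmetic that drives its trail particle effects.
; On full deployment in front-end preview, it overrides the trail's swirl and spawn rates and zeroes velocity.
; Trail updates with enough forward movement while deployed enable particles and color after a delay, with a timeline fading ColorAlpha in.
; Otherwise particles and color switch off, and the timeline reverses if deployed or has its time set if not.
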(event EventOnGliderFullyDeployed
  (bind _returnvalue (CallFunction|GetTrailVFXComponent))
  (|CallParentFunction)
  (if (Variables|FortPlayerParachute|GetbIsFrontEndPreview)
    (Effects|Components|ParticleSystem|SetFloatParameter _returnvalue "FrontEnd_Swirl_SpawnRate" 4.0)
    (Effects|Components|ParticleSystem|SetFloatParameter _returnvalue "FrontEnd_Trail_SpawnRate" 3e+01)
    (Effects|Components|ParticleSystem|SetVectorParameter _returnvalue "FrontEnd_Velocity" "0, 0, 0")))

(event EventOnChuteVisibilityChanged
  (params (bParachuteVisibility :type Boolean))
  (|CallParentFunction bParachuteVisibility)
  (if bParachuteVisibility
    _))

(event Parachute|EventOnParachuteTrailUpdated
  (params (MovementLengthSquared :type Float%20%28single-precision%29) (ForwardDot :type Float%20%28single-precision%29) (RightDot :type Float%20%28single-precision%29) (RotationalMovementAcceleration :type Float%20%28single-precision%29))
  (bind _trailcomponents (Variables|FortPlayerParachute|GetTrailComponents))
  (bind _output (Utilities|Array|Get(acopy) _trailcomponents))
  (bind _trailcomponents_2 (Variables|FortPlayerParachute|GetTrailComponents))
  (bind _output_2 (Utilities|Array|Get(acopy) _trailcomponents_2))
  (|CallParentFunction MovementLengthSquared ForwardDot RightDot RotationalMovementAcceleration)
  (if (and (>= ForwardDot 0.2) (CallFunction|IsGliderFullyDeployedAccordingtoAnim))
    (Utilities|FlowControl|Delay 1.0)
    (Effects|Components|ParticleSystem|SetBoolParameter _output_2 "Particle Spawn Switch" true)
    (Effects|Components|ParticleSystem|SetBoolParameter _output_2 "Color On?" true)
    (bind (_direction _color_alpha) (Timeline|Alpha%20Timeline Play)
      (:Update
        (Effects|Components|ParticleSystem|SetFloatParameter (Utilities|Array|Get(acopy) (Variables|FortPlayerParachute|GetTrailComponents)) "ColorAlpha" _color_alpha)))
    (else
      (Effects|Components|ParticleSystem|SetBoolParameter _output "Particle Spawn Switch" false)
      (Effects|Components|ParticleSystem|SetBoolParameter _output "Color On?" false)
      (if (CallFunction|IsGliderFullyDeployedAccordingtoAnim)
        (Timeline|Alpha%20Timeline Reverse)
        (else
          (Timeline|Alpha%20Timeline SetNewTime))))))

(fn ConstructionScript
  (|CallParentFunction))
