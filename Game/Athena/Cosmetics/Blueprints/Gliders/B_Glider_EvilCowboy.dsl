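; Evil Cowboy glider Blueprint that drives a locomotive-themed parachute's flames, exhaust, chimney, wheels, lantern, coal, whistle and audio.
; Trail updates pass movement dots into control events that steer flame materials, bursts, wheel speed and sound.
; An anim-based deployment check gates effects, with DoOnce-guarded timelines fading material visibility in and out.
; The anim Blueprint's FEIdleStart replays the controls for cosmetic previews, while CaptureTeam forces visuals for UI captures.
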(event Custom|CaptureTeam
  (bind _frontendforwarddot (Variables|Default|GetFrontEndForwardDot))
  (bind _trailparticles (Variables|Parachute|GetTrailParticles))
  (CallFunction|CaptureFlames _frontendforwarddot)
  (CallFunction|CaptureOther _frontendforwarddot)
  (Effects|Components|ParticleSystem|SetVectorParameter _trailparticles "Locomotive_TestVelocity_L" "0.000000,200.000000,0.000000")
  (Effects|Components|ParticleSystem|SetVectorParameter _trailparticles "Locomotive_TestVelocity_R" "0.000000,200.000000,0.000000")
  (Effects|Components|ParticleSystem|SetVectorParameter _trailparticles "Locomotive_TestVel_L" "0.000000,25.000000,0.000000")
  (Effects|Components|ParticleSystem|SetVectorParameter _trailparticles "Locomotive_TestVel_R" "0.000000,-25.000000,0.000000")
  (Effects|Components|ParticleSystem|SetFloatParameter _trailparticles "FESpawn" 5.0))

(event Custom|FrontEndTest
  (bind _frontendforwarddot (Variables|Default|GetFrontEndForwardDot))
  (bind _trailparticles (Variables|Parachute|GetTrailParticles))
  (if (Variables|FortPlayerParachute|GetIsCosmeticPreview)
    (CallFunction|TrainFlamesSetup _frontendforwarddot)
    (CallFunction|ExhaustControl _frontendforwarddot)
    (CallFunction|WheelSpeedControl _frontendforwarddot)
    (CallFunction|AudioUpdate _frontendforwarddot)
    (CallFunction|ChimneyControl _frontendforwarddot)
    (CallFunction|LanternControl)
    (CallFunction|CoalControl)
    (CallFunction|WhistleControl)
    (Effects|Components|ParticleSystem|SetVectorParameter _trailparticles "Locomotive_TestVelocity_L" "200.000000, 0,0.000000")
    (Effects|Components|ParticleSystem|SetVectorParameter _trailparticles "Locomotive_TestVelocity_R" "200.000000, 0,0.000000")
    (Effects|Components|ParticleSystem|SetVectorParameter _trailparticles "Locomotive_TestVel_L" "0.000000,25.000000,0.000000")
    (Effects|Components|ParticleSystem|SetVectorParameter _trailparticles "Locomotive_TestVel_R" "0.000000,-25.000000,0.000000")
    (Effects|Components|ParticleSystem|SetFloatParameter _trailparticles "FESpawn" 5.0)))

(event EventBeginPlay
  (|CallParentFunction)
  (bind _asbad_ass_cowboy_glider_anim_bp (Utilities|Casting|CastToBadAssCowboy_Glider_AnimBP (Components|SkeletalMesh|GetAnimInstance (Variables|Parachute|GetParachuteMesh))))
  (Default|BindEventtoFEIdleStart _asbad_ass_cowboy_glider_anim_bp (EventDispatchers|CreateEvent :!event Custom|FrontEndTest)))

(event EventDestroyed
  (|CallParentFunction)
  (bind _asbad_ass_cowboy_glider_anim_bp (Utilities|Casting|CastToBadAssCowboy_Glider_AnimBP (Components|SkeletalMesh|GetAnimInstance (Variables|Parachute|GetParachuteMesh))))
  (Default|UnbindallEventsfromFEIdleStart _asbad_ass_cowboy_glider_anim_bp))

(event Parachute|EventOnParachuteTrailUpdated
  (params (MovementLengthSquared :type Float%20%28single-precision%29) (ForwardDot :type Float%20%28single-precision%29) (RightDot :type Float%20%28single-precision%29) (RotationalMovementAcceleration :type Float%20%28single-precision%29))
  (CallFunction|TrainFlamesSetup ForwardDot RightDot)
  (CallFunction|ExhaustControl ForwardDot)
  (CallFunction|WheelSpeedControl ForwardDot)
  (CallFunction|AudioUpdate ForwardDot)
  (CallFunction|ChimneyControl ForwardDot)
  (CallFunction|LanternControl)
  (CallFunction|CoalControl)
  (CallFunction|WhistleControl))

(event Custom|Whistle%20Control
  ;[0 0.7816 0.88] Steam Whistle
  (if (CallFunction|IsGliderFullyDeployedAccordingtoAnim)
    _)
  (Utilities|FlowControl|DoOnce false
    (:Completed
      (Utilities|FlowControl|Delay 0.2)
      (Components|Activation|Activate (Variables|Default|GetWhistle) false))))

(event Custom|CaptureFlames
  (params (Forward%20Dot :type Float%20%28double-precision%29) (Right%20Dot :type Float%20%28double-precision%29))
  ;[0.4484 0 1] Capture Setup for UI Captures
  (bind _output_get (Variables|Default|SetCurrentFlame (Math|Interpolation|VInterpTo (Variables|Default|GetCurrentFlame) (CallFunction|DetermineFlameMovement Forward%20Dot Right%20Dot) (Utilities|Time|GetWorldDeltaSeconds) 2.0)))
  (Rendering|Material|SetVectorParameterValueonMaterials (Variables|Default|GetFlame%20Meshes) "LocalPostioning" _output_get)
  (bind _output_get_2 (Variables|Default|SetCurrentExhaust (Math|Interpolation|VInterpTo (Variables|Default|GetCurrentExhaust) (CallFunction|DetermineExhaustFlameMovement Forward%20Dot Right%20Dot) (Utilities|Time|GetWorldDeltaSeconds) 2.0)))
  (Rendering|Material|SetVectorParameterValueonMaterials (Variables|Default|GetExhaust_L) "LocalPostioning" _output_get_2)
  (Rendering|Material|SetScalarParameterValueonMaterials (Variables|Default|GetFlame%20Meshes) "OpacityMaskMultiplier" 1e+02))

(event Custom|Train%20Flames%20Setup
  (params (Forward%20Dot :type Float%20%28double-precision%29) (Right%20Dot :type Float%20%28double-precision%29))
  ;[0 0.7816 0.88] Setups the Flames coming from the Train
  (bind _output_get (Variables|Default|SetCurrentFlame (Math|Interpolation|VInterpTo (Variables|Default|GetCurrentFlame) (CallFunction|DetermineFlameMovement Forward%20Dot Right%20Dot) (Utilities|Time|GetWorldDeltaSeconds) 2.0)))
  (Rendering|Material|SetVectorParameterValueonMaterials (Variables|Default|GetFlame%20Meshes) "LocalPostioning" _output_get)
  (bind _output_get_2 (Variables|Default|SetCurrentExhaust (Math|Interpolation|VInterpTo (Variables|Default|GetCurrentExhaust) (CallFunction|DetermineExhaustFlameMovement Forward%20Dot Right%20Dot) (Utilities|Time|GetWorldDeltaSeconds) 2.0)))
  (Rendering|Material|SetVectorParameterValueonMaterials (Variables|Default|GetExhaust_L) "LocalPostioning" _output_get_2)
  ;[0 0.7816 0.88] Flames Visibility Control
  (if (CallFunction|IsGliderFullyDeployedAccordingtoAnim)
    (Utilities|FlowControl|DoOnce false
      (:Completed
        (bind (_direction _lerp) (Timeline|Flames%20Lerp%20Controller PlayFromStart)
          (:Update
            (Rendering|Material|SetScalarParameterValueonMaterials (Variables|Default|GetFlame%20Meshes) "OpacityMaskMultiplier" (Math|Float|Lerp 0.0 1e+02 _lerp)))
          (:Finished
            (switch Utilities|FlowControl|Switch|SwitchonETimelineDirection _direction
              (:Forward
                (CallFunction|ResetReverseFlames))
              (:Backward
                (CallFunction|ResetPlayFlames)))))))
    (else
      (Utilities|FlowControl|DoOnce true
        (:Completed
          (Timeline|Flames%20Lerp%20Controller ReverseFromEnd))))))

(event Custom|Audio%20Update
  (params (Forward%20Dot :type Float%20%28double-precision%29))
  (bind _fortvehicleaudiovoice (Variables|Default|GetFortVehicleAudioVoice))
  ;[0 0 0] Audio logic
  (Variables|Default|SetForwardDotCached Forward%20Dot)
  (if (CallFunction|IsGliderFullyDeployedAccordingtoAnim)
    (Class|FortVehicleAudioVoice|SetParam _fortvehicleaudiovoice (select (> Forward%20Dot 0.2) 1.0 0.5))
    (else
      (Class|FortVehicleAudioVoice|SetParam _fortvehicleaudiovoice 0.0))))

(event Custom|Reset_Play_Flames
  ;[0 0.7816 0.88] Flames Visibility Control
  (Utilities|FlowControl|DoOnce false
    (:Completed
      (Timeline|Flames%20Lerp%20Controller PlayFromStart))))

(event Custom|CaptureOther
  (params (Forward%20Dot :type Float%20%28double-precision%29))
  ;[0.4484 0 1] Capture Setup for UI Captures
  (Components|Activation|Activate (Variables|Default|GetFX_ExhaustFlames_R) false)
  (Rendering|Material|SetScalarParameterValueonMaterials (Variables|Default|GetExhaust_R) "OpacityMaskMultiplier" 5e+01)
  (Rendering|Material|SetScalarParameterValueonMaterials (Variables|Default|GetWheel_A_R) "Opacity" 1.5)
  (Rendering|Material|SetScalarParameterValueonMaterials (Variables|Default|GetWheel_A_L) "Speed" 1e+01))

(event Custom|Reset_Reverse_Flames
  ;[0 0.7816 0.88] Flames Visibility Control
  (Utilities|FlowControl|DoOnce true
    (:Completed
      (Timeline|Flames%20Lerp%20Controller ReverseFromEnd))))

(event Custom|On%20Wheel%20Rotation%20Complete
  (bind _forwarddotcached (Variables|Default|GetForwardDotCached))
  (bind _returnvalue (> _forwarddotcached 0.2))
  (bind _wheelturnvoiceindex (Variables|Default|GetWheelTurnVoiceIndex))
  (bind _wheelturnaudio (Variables|Default|GetWheelTurnAudio))
  (bind _output (Utilities|Array|Get(acopy) _wheelturnaudio _wheelturnvoiceindex))
  ;[0 0 0] Wheel Rotation Audio
  (Utilities|FlowControl|Sequence
    (:then_0
      (Audio|Components|Audio|Play _output 0.0)
      (Audio|Components|Audio|SetPitchMultiplier _output (select _returnvalue 1.4 (select (< (Variables|Default|GetForwardDotCached) 0.0) 1.2 1.0)))
      (Audio|Components|Audio|SetVolumeMultiplier _output (select _returnvalue 0.6 1.0) :!class /Script/Engine.AudioComponent))
    (:then_1
      (Variables|Default|SetWheelTurnVoiceIndex (Math|Integer|%25%28Integer%29 (+ (Variables|Default|GetWheelTurnVoiceIndex) 1) (Utilities|Array|Length (Variables|Default|GetWheelTurnAudio)))))))

(event Custom|Reset_Chimney
  ;[0 0.7816 0.88] Chimney Control
  (Utilities|FlowControl|DoOnce false
    (:Completed
      (Components|Activation|Activate (Variables|Default|GetChimney%20Burst) false))))

(event Custom|Reset_Exhaust
  ;[0 0.7816 0.88] Exhaust Control
  (Utilities|FlowControl|DoOnce false
    (:Completed
      (Components|Activation|Activate (Variables|Default|GetFX_ExhaustFlames_L_Burst) false))))

(event Custom|Chimney%20Control
  (params (Foward%20Dot :type Float%20%28double-precision%29))
  (bind _chimneyfire_fx (Variables|Default|GetChimneyFire%20FX))
  ;[0 0.7816 0.88] Chimney Control
  (if (CallFunction|IsGliderFullyDeployedAccordingtoAnim)
    (Components|Activation|Activate _chimneyfire_fx false)
    (Utilities|FlowControl|Sequence
      (:then_0
        (if (>= Foward%20Dot 0.2)
          (Utilities|FlowControl|DoOnce false
            (:Completed
              (Components|Activation|Activate (Variables|Default|GetChimney%20Burst) false)))
          (else
            (CallFunction|ResetChimney))))
      (:then_1
        (Utilities|FlowControl|DoOnce false
          (:Completed
            (bind (_direction _lerp) (Timeline|Chimney%20Lerp%20Controller PlayFromStart)
              (:Update
                (Rendering|Material|SetScalarParameterValueonMaterials (Variables|Default|GetChimney%20Flames) "OpacityMaskMultiplier" (Math|Float|Lerp 0.0 5e+01 _lerp)))
              (:Finished
                (switch Utilities|FlowControl|Switch|SwitchonETimelineDirection _direction
                  (:Forward
                    (CallFunction|ResetReverseChimney))
                  (:Backward
                    (CallFunction|ResetPlayChimney)))))))))
    (else
      (Components|Activation|Deactivate _chimneyfire_fx)
      (CallFunction|ResetChimney)
      (Utilities|FlowControl|DoOnce true
        (:Completed
          (Timeline|Chimney%20Lerp%20Controller ReverseFromEnd))))))

(event Custom|Exhaust%20Control
  (params (Forward%20Dot :type Float%20%28double-precision%29))
  (bind _fx_exhaustflames_r (Variables|Default|GetFX_ExhaustFlames_R))
  ;[0 0.7816 0.88] Exhaust Control
  (if (CallFunction|IsGliderFullyDeployedAccordingtoAnim)
    (Components|Activation|Activate _fx_exhaustflames_r false)
    (Utilities|FlowControl|Sequence
      (:then_0
        (if (>= Forward%20Dot 0.2)
          (Utilities|FlowControl|DoOnce false
            (:Completed
              (Components|Activation|Activate (Variables|Default|GetFX_ExhaustFlames_L_Burst) false)))
          (else
            (CallFunction|ResetExhaust))))
      (:then_1
        (Utilities|FlowControl|DoOnce false
          (:Completed
            (bind (_direction _lerp) (Timeline|Exhaust%20Lerp%20Controller PlayFromStart)
              (:Update
                (Rendering|Material|SetScalarParameterValueonMaterials (Variables|Default|GetExhaust_R) "OpacityMaskMultiplier" (Math|Float|Lerp 0.0 5e+01 _lerp)))
              (:Finished
                (switch Utilities|FlowControl|Switch|SwitchonETimelineDirection _direction
                  (:Forward
                    (CallFunction|ResetReverse))
                  (:Backward
                    (CallFunction|ResetPlay)))))))))
    (else
      (Components|Activation|Deactivate _fx_exhaustflames_r)
      (CallFunction|ResetExhaust)
      (Utilities|FlowControl|DoOnce true
        (:Completed
          (Timeline|Exhaust%20Lerp%20Controller ReverseFromEnd))))))

(event Custom|Reset_Play_Chimney
  ;[0 0.7816 0.88] Chimney Control
  (Utilities|FlowControl|DoOnce false
    (:Completed
      (Timeline|Chimney%20Lerp%20Controller PlayFromStart))))

(event Custom|Reset_Play
  ;[0 0.7816 0.88] Exhaust Control
  (Utilities|FlowControl|DoOnce false
    (:Completed
      (Timeline|Exhaust%20Lerp%20Controller PlayFromStart))))

(event Custom|Reset_Reverse_Chimney
  ;[0 0.7816 0.88] Chimney Control
  (Utilities|FlowControl|DoOnce true
    (:Completed
      (Timeline|Chimney%20Lerp%20Controller ReverseFromEnd))))

(event Custom|Reset_Reverse
  ;[0 0.7816 0.88] Exhaust Control
  (Utilities|FlowControl|DoOnce true
    (:Completed
      (Timeline|Exhaust%20Lerp%20Controller ReverseFromEnd))))

(event Custom|Wheel%20Speed%20Control
  (params (Forwar%20Dot :type Float%20%28double-precision%29))
  ;[0 0.7816 0.88] Wheel FX Cloud Speed and Rotation Control
  (Rendering|Material|SetScalarParameterValueonMaterials (Variables|Default|GetWheel_A_L) "Speed" (Math|Float|Lerp (select (> Forwar%20Dot 0.0) 1e+01 -1e+01) (select (> Forwar%20Dot 0.0) 3e+01 -3e+01) (Math|Float|MapRangeClamped Forwar%20Dot -1.0 1.0 0.0 1.0)))
  ;[0 0.7816 0.88] Wheel Visibility
  (if (CallFunction|IsGliderFullyDeployedAccordingtoAnim)
    (Utilities|FlowControl|Sequence
      (:then_0
        (Utilities|FlowControl|DoOnce false
          (:Completed
            (Utilities|FlowControl|Delay 1.0)
            (Components|Activation|Activate (Variables|Parachute|GetTrailParticles) false))))
      (:then_1
        (Utilities|FlowControl|DoOnce false
          (:Completed
            (bind (_direction _lerp) (Timeline|Timeline_0 PlayFromStart)
              (:Update
                (Rendering|Material|SetScalarParameterValueonMaterials (Variables|Default|GetWheel_A_R) "Opacity" (Math|Float|Lerp 0.0 1.5 _lerp)))
              (:Finished
                (switch Utilities|FlowControl|Switch|SwitchonETimelineDirection _direction
                  (:Forward
                    (CallFunction|ResetReverseWheels))
                  (:Backward
                    (CallFunction|ResetPlayWheels)))))))))
    (else
      (Components|Activation|Deactivate (Variables|Parachute|GetTrailParticles))
      (CallFunction|ResetWheels)
      (Utilities|FlowControl|DoOnce true
        (:Completed
          (Timeline|Timeline_0 ReverseFromEnd))))))

(event Custom|Reset%20Wheels
  ;[0 0.7816 0.88] Wheel Visibility
  (Utilities|FlowControl|DoOnce false
    (:Completed
      (Utilities|FlowControl|Delay 1.0)
      (Components|Activation|Activate (Variables|Parachute|GetTrailParticles) false))))

(event Custom|Lantern%20Control
  (bind _lantern_fx (Variables|Default|GetLantern%20FX))
  ;[0 0.7816 0.88] Lantern Control
  (if (CallFunction|IsGliderFullyDeployedAccordingtoAnim)
    (Components|Activation|Activate _lantern_fx false)
    (else
      (Components|Activation|Deactivate _lantern_fx))))

(event Custom|Reset_Play_Wheels
  ;[0 0.7816 0.88] Wheel Visibility
  (Utilities|FlowControl|DoOnce false
    (:Completed
      (Timeline|Timeline_0 PlayFromStart))))

(event Custom|Coal%20Control
  (bind _coalbucket (Variables|Default|GetCoalBucket))
  ;[0 0.7816 0.88] Coal Control
  (if (CallFunction|IsGliderFullyDeployedAccordingtoAnim)
    (Components|Activation|Activate _coalbucket false)
    (else
      (Components|Activation|Deactivate _coalbucket))))

(event Custom|Reset_Reverse_Wheels
  ;[0 0.7816 0.88] Wheel Visibility
  (Utilities|FlowControl|DoOnce true
    (:Completed
      (Timeline|Timeline_0 ReverseFromEnd))))

(fn DetermineExhaustFlameMovement
  (params (Forward%20Dot :type Float%20%28double-precision%29) (Right%20Dot :type Float%20%28double-precision%29))
  (return (+ (* (Math|Vector|Lerp(Vector) "0,0.000000,-1.000000" "0.000000,0.000000,1.000000" (Math|Float|MapRangeClamped Forward%20Dot -1.0 1.0 0.0 1.0)) (Variables|Default|GetForward_Intensity)) (* (Math|Vector|Lerp(Vector) "0.000000,1.000000,0.000000" "0.000000,-1.000000, 0" (Math|Float|MapRangeClamped Right%20Dot -1.0 1.0 0.0 1.0)) (Variables|Default|GetRight_Intensity)))))

(fn DetermineFlameMovement
  (params (Forward%20Dot :type Float%20%28double-precision%29) (Right%20Dot :type Float%20%28double-precision%29))
  (return (+ (* (Math|Vector|Lerp(Vector) "0,-1.000000, 0" "0,1.000000, 0" (Math|Float|MapRangeClamped Forward%20Dot -1.0 1.0 0.0 1.0)) (Variables|Default|GetForward_Intensity)) (* (Math|Vector|Lerp(Vector) "1.000000,0.000000, 0" "-1.000000,0.000000, 0" (Math|Float|MapRangeClamped Right%20Dot -1.0 1.0 0.0 1.0)) (Variables|Default|GetRight_Intensity)))))

(fn ConstructionScript
  (Utilities|Array|Add (Variables|Default|GetWheelTurnAudio) (Variables|Default|GetWheelTurn1))
  (Utilities|Array|Add (Variables|Default|GetWheelTurnAudio) (Variables|Default|GetWheelTurn2))
  (Utilities|Array|Add (Variables|Default|GetWheelTurnAudio) (Variables|Default|GetWheelTurn3)))
